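# First-person sound cue for the LlamaFloat emote, routed through the 1P emote music sound class and concurrency, with a music analysis submix send.
# A mount one-shot plays immediately, mixed with a briefly delayed group of continuous layers.
# That group blends an enveloped loop that fades in with two looping movement layers, A and B.
# Each movement layer picks random variations after random short delays, and B starts offset from A.
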SoundCue /BRCosmetics/Sounds/Emotes/LlamaFloat/Emote_LlamaFloat_1P_Cue
root=Mixer_0
nodes [
 {id=Delay_0, children [Mixer_1], properties {delayMax=0.3, delayMin=0.3}}
 {id=Delay_1, children [Random_2], properties {delayMax=0.5}}
 {id=Delay_2, children [Random_3], properties {delayMax=0.5}}
 {id=Delay_3, children [Looping_1], properties {delayMax=0.5, delayMin=0.5}}
 {
  id=Enveloper_1
  children [WavePlayer_1]
  properties {pitchCurve {editorCurveData {keys [{time=0}]}}, volumeCurve {editorCurveData {keys [<at 0 0>,<at 1 1>]}}}
 }
 {id=Looping_0, children [Delay_1]}
 {id=Looping_1, children [Delay_2]}
 {id=Mixer_0, children [WavePlayer_0,Delay_0], properties {inputVolume [1,1]}}
 {id=Mixer_1, children [Enveloper_1,Looping_0,Delay_3], properties {inputVolume [1,1,1]}}
 {id=Random_2, children [WavePlayer_2,WavePlayer_3,WavePlayer_4,WavePlayer_5], properties {weights [1,1,1,1]}}
 {id=Random_3, children [WavePlayer_6,WavePlayer_7,WavePlayer_8], properties {weights [1,1,1]}}
 {id=WavePlayer_0, properties {soundWaveAssetPtr=&/Emote_LlamaFloat_Mount}}
 {id=WavePlayer_1, properties {bLooping=true, soundWaveAssetPtr=&/Emote_LlamaFloat_LP_01}}
 {id=WavePlayer_2, properties {soundWaveAssetPtr=&/Emote_LlamaFloat_Mvmt_A_01}}
 {id=WavePlayer_3, properties {soundWaveAssetPtr=&/Emote_LlamaFloat_Mvmt_A_02}}
 {id=WavePlayer_4, properties {soundWaveAssetPtr=&/Emote_LlamaFloat_Mvmt_A_03}}
 {id=WavePlayer_5, properties {soundWaveAssetPtr=&/Emote_LlamaFloat_Mvmt_A_04}}
 {id=WavePlayer_6, properties {soundWaveAssetPtr=&/Emote_LlamaFloat_Mvmt_B_02}}
 {id=WavePlayer_7, properties {soundWaveAssetPtr=&/Emote_LlamaFloat_Mvmt_B_01}}
 {id=WavePlayer_8, properties {soundWaveAssetPtr=&/Emote_LlamaFloat_Mvmt_B_03}}
]
properties {
 maxDistance=2097152
 soundClassObject=/Game/Sounds/FortSoundClassesAndModes/Fort_Music_Emote_1P_PSM
 volumeMultiplier=1
 concurrencySet [/Game/Sounds/Concurrency_Rules/Athena_EmoteMusic_1P_Conc]
 soundSubmixSends [{minSendLevel=0, soundSubmix=&Fortnite_AudioAnalysis_Music}]
}
