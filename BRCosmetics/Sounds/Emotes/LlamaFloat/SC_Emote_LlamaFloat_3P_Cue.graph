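# Third-person sound cue for the LlamaFloat emote, routed as emote music with attenuation, concurrency rules and an audio-analysis submix send.
# A mount one-shot plays immediately, mixed with a briefly delayed group of three sustained layers.
# These are a looping bed faded in by an enveloper and two loops of randomly chosen movement sounds with randomized delays, the second starting later.
# The cue keeps playing virtualized when silent.
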SoundCue /BRCosmetics/Sounds/Emotes/LlamaFloat/SC_Emote_LlamaFloat_3P_Cue
root=Mixer_3
nodes [
 {id=Delay_4, children [Mixer_4], properties {delayMax=0.3, delayMin=0.3}}
 {id=Delay_5, children [Random_0], properties {delayMax=0.5}}
 {id=Delay_6, children [Random_1], properties {delayMax=0.5}}
 {id=Delay_7, children [Looping_3], properties {delayMax=0.5, delayMin=0.5}}
 {
  id=Enveloper_1
  children [WavePlayer_12]
  properties {pitchCurve {editorCurveData {keys [{time=0}]}}, volumeCurve {editorCurveData {keys [<at 0 0>,<at 1 1>]}}}
 }
 {id=Looping_2, children [Delay_5]}
 {id=Looping_3, children [Delay_6]}
 {id=Mixer_3, children [WavePlayer_11, Delay_4], properties {inputVolume [1,1]}}
 {id=Mixer_4, children [Enveloper_1,Looping_2,Delay_7], properties {inputVolume [1,1,1]}}
 {id=Random_0, children [WavePlayer_13, WavePlayer_14, WavePlayer_15, WavePlayer_16], properties {weights [1,1,1,1]}}
 {id=Random_1, children [WavePlayer_17, WavePlayer_18, WavePlayer_19], properties {weights [1,1,1]}}
 {id=WavePlayer_11, properties {soundWaveAssetPtr=&/Emote_LlamaFloat_Mount}}
 {id=WavePlayer_12, properties {bLooping=true, soundWaveAssetPtr=&/Emote_LlamaFloat_LP_01}}
 {id=WavePlayer_13, properties {soundWaveAssetPtr=&/Emote_LlamaFloat_Mvmt_A_01}}
 {id=WavePlayer_14, properties {soundWaveAssetPtr=&/Emote_LlamaFloat_Mvmt_A_02}}
 {id=WavePlayer_15, properties {soundWaveAssetPtr=&/Emote_LlamaFloat_Mvmt_A_03}}
 {id=WavePlayer_16, properties {soundWaveAssetPtr=&/Emote_LlamaFloat_Mvmt_A_04}}
 {id=WavePlayer_17, properties {soundWaveAssetPtr=&/Emote_LlamaFloat_Mvmt_B_02}}
 {id=WavePlayer_18, properties {soundWaveAssetPtr=&/Emote_LlamaFloat_Mvmt_B_01}}
 {id=WavePlayer_19, properties {soundWaveAssetPtr=&/Emote_LlamaFloat_Mvmt_B_03}}
]
properties {
 attenuationSettings=&Fort_Emotes_Athena_Att
 maxDistance=1600
 soundClassObject=/Game/Sounds/FortSoundClassesAndModes/Fort_Music_Emote_3P
 virtualizationMode=PlayWhenSilent
 volumeMultiplier=1
 concurrencySet [/Game/Sounds/Concurrency_Rules/Athena_EmoteMusic_Conc]
 soundSubmixSends [{minSendLevel=0, soundSubmix=&Fortnite_AudioAnalysis_Music}]
}
